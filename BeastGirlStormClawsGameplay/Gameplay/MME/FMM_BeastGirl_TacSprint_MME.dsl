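; FMM_BeastGirl_TacSprint_MME is a gameplay Blueprint, apparently for BeastGirl's tactical sprint, that overrides whether a holster can be applied.
; Its only function, CanApplyHolster, receives a holster ID by reference and always returns false, so no holster is permitted.
; The file shows no other logic, variables or components.
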(fn CanApplyHolster
  (params (HolsterId :type Name%20%28by%20ref%29 :default "None"))
  (return false))
