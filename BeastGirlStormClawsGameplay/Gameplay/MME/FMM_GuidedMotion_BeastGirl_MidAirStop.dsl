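; Fortnite guided-motion movement mode extension for Beast Girl's mid-air stop, applying a constant force set when it activates.
; On activation it tilts the owner's forward vector 35 degrees around its right axis, scales it by InitialUpForce, adds current velocity and stores the result as BaseVelocity, capped at MasSpeedForce.
; The constant-force direction returns that stored vector, clamped again rather than normalized, while strength defers to the parent.
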(event EventOnExtensionActivate
  (params (RuntimeContext :type MMERuntime%20Context%20Structure%20%28by%20ref%29))
  (bind _ownerpawn (Variables|FortMovementMode_BaseExtLogic|GetOwnerPawn))
  (|CallParentFunction RuntimeContext)
  (Utilities|IsValid _ownerpawn
    (:Is%20Valid
      (Variables|Default|SetBaseVelocity (Math|Vector|ClampVectorSize (+ (* (Math|Vector|RotateVectorAroundAxis (Transformation|GetActorForwardVector _ownerpawn) -35.0 (Transformation|GetActorRightVector _ownerpawn)) (Variables|Default|GetInitialUpForce)) (Transformation|GetVelocity _ownerpawn)) 0.0 (Variables|Default|GetMasSpeedForce))))))

(fn GetDirection_ConstantForce
  (params (GuidedMotionRuntimeData :type Fort%20Movement%20Mode%20Guided%20Motion%20Runtime%20Data%20Object%20Reference))
  (return (Math|Vector|ClampVectorSize (Variables|Default|GetBaseVelocity) 0.0 (Variables|Default|GetMasSpeedForce))))

(fn GetStrength_ConstantForce
  (params (GuidedMotionRuntimeData :type Fort%20Movement%20Mode%20Guided%20Motion%20Runtime%20Data%20Object%20Reference))
  (bind _returnvalue (|CallParentFunction GuidedMotionRuntimeData))
  (return _returnvalue))
