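; Guided motion movement mode for BeastGirl's leap, computing its launch velocity.
; CalculateLeap targets a point along networked movement input (actor forward if none), raised by a fixed height, with the TacticalSprint tag selecting the distance multiplier and arc.
; GetLaunchVelocity solves a custom-arc projectile velocity to that point and adds half the current velocity.
; GetSyncronizedMovementInput rebuilds the input vector from three networked optional floats, defaulting to zero.
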(fn CalculateLeap
  (params (RuntimeData :type Fort%20Movement%20Mode%20Guided%20Motion%20Runtime%20Data%20Object%20Reference))
  (return (Transformation|GetActorLocation (Variables|FortMovementMode_BaseExtLogic|GetOwnerPawn)) (+ (+ (Transformation|GetActorLocation (Variables|FortMovementMode_BaseExtLogic|GetOwnerPawn)) (* (Math|Vector|SelectVector (Helper|GetSyncronizedMovementInput RuntimeData) (Transformation|GetActorForwardVector (Variables|FortMovementMode_BaseExtLogic|GetOwnerPawn)) (> (Math|Vector|VectorLength (Helper|GetSyncronizedMovementInput RuntimeData)) 0)) (Math|Float|SelectFloat (ScalableFloat|GetValueatLevel (Variables|Default|GetHF_TacSprintLeapMult)) (ScalableFloat|GetValueatLevel (Variables|Default|GetHF_ForwardLeapMult)) (GameplayTags|HasMatchingGameplayTag (Utilities|Casting|ToGameplayTagAssetInterface (Variables|FortMovementMode_BaseExtLogic|GetOwnerPawn)) "(TagName=\"Gameplay.Action.BeastGirl.TacticalSprint\")")))) "0.000000,0.000000,250.000000") (Math|Float|SelectFloat 0.5 0.3 (GameplayTags|HasMatchingGameplayTag (Utilities|Casting|ToGameplayTagAssetInterface (Variables|FortMovementMode_BaseExtLogic|GetOwnerPawn)) "(TagName=\"Gameplay.Action.BeastGirl.TacticalSprint\")"))))

(fn GetLaunchVelocity
  (params (Runtime%20Data :type Fort%20Movement%20Mode%20Guided%20Motion%20Runtime%20Data%20Object%20Reference))
  (bind (_startpos _endpos _arc) (Initialization|CalculateLeap Runtime%20Data))
  (bind _outlaunchvelocity (Game|SuggestProjectileVelocityCustomArc (Variables|FortMovementMode_BaseExtLogic|GetOwnerPawn) _startpos _endpos 0.0 _arc))
  (return (+ _outlaunchvelocity (/ (Variables|/Script/Engine.MovementComponent|GetVelocity (Variables|FortMovementMode_BaseExtLogic|GetOwnerMovementComponent)) 2.0))))

(fn GetSyncronizedMovementInput
  (params (RuntimeData :type Fort%20Movement%20Mode%20Guided%20Motion%20Runtime%20Data%20Object%20Reference))
  (bind _netoptionalfloats (Variables|/Script/FortniteGame.FortMovementMode_GuidedMotionRuntimeData|GetNetOptionalFloats RuntimeData))
  (bind _output_value (Utilities|Struct|BreakFortMovementModeFloatPropertyNetworked (Utilities|Array|Get(aref) _netoptionalfloats 2)))
  (bind _output_value_2 (Utilities|Struct|BreakFortMovementModeFloatPropertyNetworked (Utilities|Array|Get(aref) _netoptionalfloats)))
  (bind _output_value_3 (Utilities|Struct|BreakFortMovementModeFloatPropertyNetworked (Utilities|Array|Get(aref) _netoptionalfloats 1)))
  (if (>= (Utilities|Array|Length _netoptionalfloats) 3)
    (return _output_value_2 _output_value_3 _output_value)
    (else
      (return "0, 0, 0"))))
