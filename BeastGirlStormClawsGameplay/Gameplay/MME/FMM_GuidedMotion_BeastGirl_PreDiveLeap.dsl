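; Guided motion movement mode for the BeastGirl Storm Claws pre-dive leap, computing the launch velocity that throws the character upward.
; GetLaunchVelocity keeps the current horizontal velocity and clamps vertical velocity to a small upward range.
; It adds one of two upward boosts, picked by the first net optional bool in the runtime data (true gives the weaker one).
; Without that bool it returns zero velocity.
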(fn GetLaunchVelocity
  (params (Runtime%20Data :type Fort%20Movement%20Mode%20Guided%20Motion%20Runtime%20Data%20Object%20Reference))
  (bind (_velocity_x _velocity_y _velocity_z) (Math|Vector|BreakVector (Variables|/Script/Engine.MovementComponent|GetVelocity_X (Variables|FortMovementMode_BaseExtLogic|GetOwnerMovementComponent))))
  (bind _netoptionalbools (Variables|/Script/FortniteGame.FortMovementMode_GuidedMotionRuntimeData|GetNetOptionalBools Runtime%20Data))
  (if (>= (Utilities|Array|Length _netoptionalbools) 1)
    (return (+ (Math|Vector|MakeVector _velocity_x _velocity_y (Math|Float|Clamp(Float) _velocity_z 0.0 2.5e+02)) (Math|Vector|SelectVector "0.000000,0.000000,1000.000000" "0.000000,0.000000,1500.000000" (Utilities|Array|Get(aref) _netoptionalbools))))
    (else
      (return "0, 0, 0"))))
