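; Guided motion movement mode extension for the downward part of BeastGirl's Storm Slam, driving the owner along its aimed direction.
; Activation stores the net-synced aim forward vector as the target direction.
; Root-motion velocity and constant-force overrides bend that direction by Z-only gravity scaled by a time-sampled curve, normalize it, and use the scalable slam velocity as speed.
; Deactivation only prints a debug string labeled as air launch.
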(event EventOnExtensionActivate
  (params (RuntimeContext :type MMERuntime%20Context%20Structure%20%28by%20ref%29))
  (|CallParentFunction RuntimeContext)
  (Variables|Default|SetTargetDirection (Math|Vector|GetForwardVector (CharacterMovement|GetNetSyncedBaseAimRotation (Variables|FortMovementMode_BaseExtLogic|GetOwnerMovementComponent)))))

(event EventOnExtensionDeactivate
  (params (RuntimeContext :type MMERuntime%20Context%20Structure%20%28by%20ref%29))
  (Development|PrintString 0 "AIR LAUNCH DEACTIVATED"))

(fn AdjustRootMotionVelocity
  (params (RuntimeContext :type MMERuntime%20Context%20Structure%20%28by%20ref%29))
  (bind _returnvalue (Class|FortMovementModeExtBlueprintLibrary|GetRuntimeDatafromContext RuntimeContext "/Script/FortniteGame.FortMovementMode_GuidedMotionRuntimeData"))
  (Utilities|IsValid _returnvalue
    (:Is%20Valid
      (return (* (Math|Vector|Normalize (+ (Variables|Default|GetTargetDirection) (* (* (Variables|/Script/Engine.CharacterMovementComponent|GetGravityDirection (Variables|FortMovementMode_BaseExtLogic|GetOwnerMovementComponent)) (Math|Curves|GetRuntimeFloatCurveValue (Variables|Default|GetGravityCurve) (Variables|/Script/FortniteGame.FortMovementMode_GuidedMotionRuntimeData|GetLocalEndElapsedTime _returnvalue))) "0.000000,0.000000,1.000000"))) (ScalableFloat|GetValueatLevel (Variables|Default|GetHF_SlamVelocity)))))
    (:Is%20Not%20Valid
      (return "0, 0, 0"))))

(fn GetDirection_ConstantForce
  (params (GuidedMotionRuntimeData :type Fort%20Movement%20Mode%20Guided%20Motion%20Runtime%20Data%20Object%20Reference))
  (return (Math|Vector|Normalize (+ (Variables|Default|GetTargetDirection) (* (* (Variables|/Script/Engine.CharacterMovementComponent|GetGravityDirection (Variables|FortMovementMode_BaseExtLogic|GetOwnerMovementComponent)) (Math|Curves|GetRuntimeFloatCurveValue (Variables|Default|GetGravityCurve) (Variables|/Script/FortniteGame.FortMovementMode_GuidedMotionRuntimeData|GetLocalEndElapsedTime GuidedMotionRuntimeData))) "0.000000,0.000000,1.000000")))))

(fn GetStrength_ConstantForce
  (params (GuidedMotionRuntimeData :type Fort%20Movement%20Mode%20Guided%20Motion%20Runtime%20Data%20Object%20Reference))
  (return (ScalableFloat|GetValueatLevel (Variables|Default|GetHF_SlamVelocity))))
