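; Movement extension for a leap impulse in Beast Girl Storm Claws gameplay, built to stay active for a single pre-movement update.
; On push it calls the parent handler, then clears an AppliedImpulse flag.
; The first update before character movement sets the flag and returns Active; the next returns Ended.
; No impulse force is computed here, only this one-shot lifetime.
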(event EventOnExtensionPush
  (params (RuntimeContext :type MMERuntime%20Context%20Structure%20%28by%20ref%29))
  (|CallParentFunction RuntimeContext)
  (Variables|Default|SetAppliedImpulse false))

(fn UpdateBeforeCharacterMovement
  (params (ElapsedTime :type Float%20%28single-precision%29) (DeltaTime :type Float%20%28single-precision%29))
  (if (Variables|Default|GetAppliedImpulse)
    (return "Ended")
    (else
      (Variables|Default|SetAppliedImpulse true)
      (return "Active"))))
